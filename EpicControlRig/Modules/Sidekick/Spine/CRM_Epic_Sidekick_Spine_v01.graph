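# Modular Control Rig spine module for Epic's Sidekick that builds and drives hip, pelvis, waist and chest controls.
# Construction resolves connectors, adds virtual bones for placement, then spawns controls, two proxies and a secondary-visibility channel.
# Forward solve blends the waist null between Spine 01 and Chest and poses joints from controls; backward solve refits main controls.
# Interaction lets proxies move Chest or Hips through stored offsets, refreshed when interaction ends.
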ControlRigBlueprint /EpicControlRig/Modules/Sidekick/Spine/CRM_Epic_Sidekick_Spine_v01
graphs [
 Multiply_1_ContainedGraph {nodes [
   n1 = RigVMFunctionEntryNode()
   Multiply = &RigVMFunction_MathDoubleMul(A=@n1.A, B=@n1.B)
   RigVMFunctionReturnNode(Result=&RigVMFunction_MathDoubleMul(B=@n1.C, A=@Multiply.Result).Result)
  ]}
 New Function_ContainedGraph {nodes [
   RigVMFunctionEntryNode()
   > RigVMFunction_Sequence = RigVMAggregateNode()
   RigVMFunctionReturnNode(ExecuteContext=@RigVMFunction_Sequence.D)
   RigUnit_GetTransform = &RigUnit_GetTransform(bInitial=true, Item=RigVMVariableNode() {variable=Connector Spine Base}.Value)
   n1 = /Script/ControlRig.RigUnit_HierarchyAddBone(
    Name=Virtual Spine 01
    Parent=(Type=Bone,Name="None")
    Space=GlobalSpace
    Transform=(Rotation=(X=0,Y=0,Z=0,W=1),Translation=(X=0,Y=0,Z=0),Scale3D=(X=1,Y=1,Z=1))
    Transform.Translation=@RigUnit_GetTransform.Transform.Translation
    ExecutePin=@RigVMFunction_Sequence.C
   ) {is=struct RigVMUnitNode}
   > RigVMDispatchNode(Element=@n1.Item, Array=RigVMVariableNode() {variable=Virtual Spine Joints}.Value) {template="DISPATCH_RigVMDispatch_ArrayAdd(io Array,in Element,out Index)"}
   > n2 = &RigUnit_HierarchyAddBone(
    Name=Virtual Spine 02
    Parent=(Type=Bone,Name="None")
    Space=GlobalSpace
    Transform=(Rotation=(X=0,Y=0,Z=0,W=1),Translation=(X=0,Y=0,Z=0),Scale3D=(X=1,Y=1,Z=1))
    Transform.Translation=&RigUnit_GetTransform(bInitial=true, Item=RigVMVariableNode() {variable=Connector Spine Mid}.Value).Transform.Translation
   )
   > RigVMDispatchNode(Element=@n2.Item, Array=RigVMVariableNode() {variable=Virtual Spine Joints}.Value) {template="DISPATCH_RigVMDispatch_ArrayAdd(io Array,in Element,out Index)"}
   > n3 = &RigUnit_HierarchyAddBone(
    Name=Virtual Spine 03
    Parent=(Type=Bone,Name="None")
    Space=GlobalSpace
    Transform=(Rotation=(X=0,Y=0,Z=0,W=1),Translation=(X=0,Y=0,Z=0),Scale3D=(X=1,Y=1,Z=1))
    Transform.Translation=&RigUnit_GetTransform(bInitial=true, Item=RigVMVariableNode() {variable=Connector Spine Tip}.Value).Transform.Translation
   )
   > RigVMDispatchNode(Element=@n3.Item, Array=RigVMVariableNode() {variable=Virtual Spine Joints}.Value) {template="DISPATCH_RigVMDispatch_ArrayAdd(io Array,in Element,out Index)"}
   n4 = &RigUnit_HierarchyAddBone(
    Name=Virtual Pelvis
    Parent=(Type=Bone,Name="root")
    Space=GlobalSpace
    Transform=(Rotation=(X=0,Y=0,Z=0,W=1),Translation=(X=0,Y=0,Z=0),Scale3D=(X=1,Y=1,Z=1))
    Transform.Translation=&RigUnit_GetTransform(bInitial=true, Item=RigVMVariableNode() {variable=Connector Spine Root}.Value).Transform.Translation
    ExecutePin=@RigVMFunction_Sequence.A
   )
   > RigVMVariableNode(Value=@n4.Item) {variable=Pelvis Joint}
  ]}
 New Function_ContainedGraph {nodes [
   n1 = RigVMFunctionEntryNode()
   n2 = &RigVMFunction_Sequence(ExecuteContext=@n1.ExecuteContext)
   RigVMFunction_Sequence = RigVMAggregateNode(ExecuteContext=@n2.A)
   n3 = RigVMVariableNode() {variable=Virtual Spine Joints}
   Get Scale Factor Profile Value_0 = RigVMFunctionReferenceNode(Param=0)
   Prepare Control Shape = RigVMFunctionReferenceNode(
    Scale Premultiplier=0.85
    Scale Factor=@"Get Scale Factor Profile Value_0".Result
    Control Shape=RigVMVariableNode() {variable=Spine Control Shapes}.Value
   )
   Try Get Name From Array = RigVMFunctionReferenceNode(Name Array=RigVMVariableNode() {variable=Spine Display Names}.Value)
   n4 = /Script/ControlRig.RigUnit_HierarchyAddControlTransform(
    InitialValue=(Rotation=(X=0,Y=0,Z=0,W=1),Translation=(X=0,Y=0,Z=0),Scale3D=(X=1,Y=1,Z=1))
    Name=Spine 01
    OffsetSpace=GlobalSpace
    Settings="(InitialSpace=LocalSpace,bUsePreferredRotationOrder=true,PreferredRotationOrder=YXZ,Limits=(LimitTranslationX=(bMinimum=False,bMaximum=False),LimitTranslationY=(bMinimum=False,bMaximum=False),LimitTranslationZ=(bMinimum=False,bMaximum=False),LimitPitch=(bM ...(truncated)"
    OffsetTransform=&RigUnit_GetTransform(
     bInitial=true
     Item=RigVMDispatchNode(Index=0, Array=@n3.Value) {template="DISPATCH_RigVMDispatch_ArrayGetAtIndex(in Array,in Index,out Element)"}.Element
    ).Transform
    Settings.Shape=@"Prepare Control Shape".Result Control Shape
    Settings.DisplayName=@"Try Get Name From Array".Result
    Parent=RigVMVariableNode() {variable=Hip Control}.Value
    ExecutePin=@RigVMFunction_Sequence.B
   ) {
    is=struct RigVMUnitNode
    template="SpawnControl::Execute(in InitialValue,in Settings,in OffsetTransform,in OffsetSpace,in Parent,in Name,out Item)"
   }
   > RigVMVariableNode(Value=@n4.Item) {variable=Spine 01 Control}
   > RigVMDispatchNode(Element=@n4.Item, Array=RigVMVariableNode() {variable=Controls}.Value) {template="DISPATCH_RigVMDispatch_ArrayAdd(io Array,in Element,out Index)"}
   n5 = RigVMVariableNode() {variable=Virtual Spine Joints}
   > n6 = &RigUnit_HierarchyAddNull(
    Name=Spine 02 Null
    Space=GlobalSpace
    Transform=&RigUnit_GetTransform(
     bInitial=true
     Item=RigVMDispatchNode(Index=1, Array=@n5.Value) {template="DISPATCH_RigVMDispatch_ArrayGetAtIndex(in Array,in Index,out Element)"}.Element
    ).Transform
    Parent=RigVMVariableNode() {variable=Pelvis Control}.Value
   )
   > RigVMVariableNode(Value=@n6.Item) {variable=Spine 02 Null}
   Get Scale Factor Profile Value_1 = RigVMFunctionReferenceNode(Param=0.5)
   Prepare Control Shape_1 = RigVMFunctionReferenceNode(
    Scale Premultiplier=1
    Scale Factor=@"Get Scale Factor Profile Value_1".Result
    Control Shape=RigVMVariableNode() {variable=Spine Control Shapes}.Value
   )
   Try Get Name From Array_1 = RigVMFunctionReferenceNode(Index=1, Name Array=RigVMVariableNode() {variable=Spine Display Names}.Value)
   > n7 = &RigUnit_HierarchyAddControlTransform(
    InitialValue=(Rotation=(X=0,Y=0,Z=0,W=1),Translation=(X=0,Y=0,Z=0),Scale3D=(X=1,Y=1,Z=1))
    Name=Spine 02
    OffsetSpace=LocalSpace
    OffsetTransform=(Rotation=(X=0,Y=0,Z=0,W=1),Translation=(X=0,Y=0,Z=0),Scale3D=(X=1,Y=1,Z=1))
    Settings="(InitialSpace=LocalSpace,bUsePreferredRotationOrder=true,PreferredRotationOrder=YXZ,Limits=(LimitTranslationX=(bMinimum=False,bMaximum=False),LimitTranslationY=(bMinimum=False,bMaximum=False),LimitTranslationZ=(bMinimum=False,bMaximum=False),LimitPitch=(bM ...(truncated)"
    Settings.Shape=@"Prepare Control Shape_1".Result Control Shape
    Settings.DisplayName=@"Try Get Name From Array_1".Result
    Parent=RigVMVariableNode() {variable=Spine 02 Null}.Value
   ) {template="SpawnControl::Execute(in InitialValue,in Settings,in OffsetTransform,in OffsetSpace,in Parent,in Name,out Item)"}
   > RigVMVariableNode(Value=(Type=Bone,Name="None")) {variable=Spine 02}
   > RigVMDispatchNode(Element=@n7.Item, Array=RigVMVariableNode() {variable=Controls}.Value) {template="DISPATCH_RigVMDispatch_ArrayAdd(io Array,in Element,out Index)"}
   n8 = RigVMVariableNode() {variable=Virtual Spine Joints}
   Get Scale Factor Profile Value_2 = RigVMFunctionReferenceNode(Param=1)
   Prepare Control Shape_2 = RigVMFunctionReferenceNode(
    Scale Premultiplier=1
    Scale Factor=@"Get Scale Factor Profile Value_2".Result
    Control Shape=RigVMVariableNode() {variable=FKs Control Shapes}.Value
   )
   Try Get Name From Array_2 = RigVMFunctionReferenceNode(Index=2, Name Array=RigVMVariableNode() {variable=Spine Display Names}.Value)
   > n9 = &RigUnit_HierarchyAddControlTransform(
    InitialValue=(Rotation=(X=0,Y=0,Z=0,W=1),Translation=(X=0,Y=0,Z=0),Scale3D=(X=1,Y=1,Z=1))
    Name=Chest
    OffsetSpace=GlobalSpace
    Settings="(InitialSpace=LocalSpace,bUsePreferredRotationOrder=true,PreferredRotationOrder=YXZ,Limits=(LimitTranslationX=(bMinimum=False,bMaximum=False),LimitTranslationY=(bMinimum=False,bMaximum=False),LimitTranslationZ=(bMinimum=False,bMaximum=False),LimitPitch=(bM ...(truncated)"
    OffsetTransform=&RigUnit_GetTransform(
     bInitial=true
     Item=RigVMDispatchNode(Index=2, Array=@n8.Value) {template="DISPATCH_RigVMDispatch_ArrayGetAtIndex(in Array,in Index,out Element)"}.Element
    ).Transform
    Settings.Shape=@"Prepare Control Shape_2".Result Control Shape
    Settings.DisplayName=@"Try Get Name From Array_2".Result
    Parent=RigVMVariableNode() {variable=Connector Parent}.Value
   ) {template="SpawnControl::Execute(in InitialValue,in Settings,in OffsetTransform,in OffsetSpace,in Parent,in Name,out Item)"}
   > RigVMVariableNode(Value=@n9.Item) {variable=Chest Control}
   > RigVMDispatchNode(Element=@n9.Item, Array=RigVMVariableNode() {variable=Controls}.Value) {template="DISPATCH_RigVMDispatch_ArrayAdd(io Array,in Element,out Index)"}
   n10 = RigVMVariableNode() {variable=Virtual Spine Joints}
   Get Scale Factor Profile Value_3 = RigVMFunctionReferenceNode(Param=0)
   Prepare Control Shape_5 = RigVMFunctionReferenceNode(
    Scale Premultiplier=1
    Scale Factor=@"Get Scale Factor Profile Value_3".Result
    Control Shape=RigVMVariableNode() {variable=FKs Control Shapes}.Value
   )
   n11 = &RigUnit_HierarchyAddControlTransform(
    InitialValue=(Rotation=(X=0,Y=0,Z=0,W=1),Translation=(X=0,Y=0,Z=0),Scale3D=(X=1,Y=1,Z=1))
    Name=Hips
    OffsetSpace=GlobalSpace
    Settings="(InitialSpace=LocalSpace,bUsePreferredRotationOrder=true,PreferredRotationOrder=YXZ,Limits=(LimitTranslationX=(bMinimum=False,bMaximum=False),LimitTranslationY=(bMinimum=False,bMaximum=False),LimitTranslationZ=(bMinimum=False,bMaximum=False),LimitPitch=(bM ...(truncated)"
    OffsetTransform=&RigUnit_GetTransform(
     bInitial=true
     Item=RigVMDispatchNode(Index=0, Array=@n10.Value) {template="DISPATCH_RigVMDispatch_ArrayGetAtIndex(in Array,in Index,out Element)"}.Element
    ).Transform
    Settings.Shape=@"Prepare Control Shape_5".Result Control Shape
    Parent=RigVMVariableNode() {variable=Connector Parent}.Value
    Settings.DisplayName=RigVMVariableNode() {variable=Hip Display Name}.Value
    ExecutePin=@RigVMFunction_Sequence.A
   ) {template="SpawnControl::Execute(in InitialValue,in Settings,in OffsetTransform,in OffsetSpace,in Parent,in Name,out Item)"}
   > RigVMVariableNode(Value=@n11.Item) {variable=Hip Control}
   Get Scale Factor Profile Value_4 = RigVMFunctionReferenceNode(Param=0)
   Prepare Control Shape_6 = RigVMFunctionReferenceNode(
    Scale Premultiplier=0.85
    Scale Factor=@"Get Scale Factor Profile Value_4".Result
    Control Shape=RigVMVariableNode() {variable=Pelvis Control Shapes}.Value
   )
   > n12 = &RigUnit_HierarchyAddControlTransform(
    InitialValue=(Rotation=(X=0,Y=0,Z=0,W=1),Translation=(X=0,Y=0,Z=0),Scale3D=(X=1,Y=1,Z=1))
    Name=Pelvis FK
    OffsetSpace=GlobalSpace
    Settings="(InitialSpace=LocalSpace,bUsePreferredRotationOrder=true,PreferredRotationOrder=YXZ,Limits=(LimitTranslationX=(bMinimum=False,bMaximum=False),LimitTranslationY=(bMinimum=False,bMaximum=False),LimitTranslationZ=(bMinimum=False,bMaximum=False),LimitPitch=(bM ...(truncated)"
    OffsetTransform=&RigUnit_GetTransform(bInitial=true, Item=RigVMVariableNode() {variable=Pelvis Joint}.Value).Transform
    Settings.Shape=@"Prepare Control Shape_6".Result Control Shape
    Settings.DisplayName=RigVMVariableNode() {variable=Pelvis FK Display Name}.Value
    Parent=@n11.Item
   ) {template="SpawnControl::Execute(in InitialValue,in Settings,in OffsetTransform,in OffsetSpace,in Parent,in Name,out Item)"}
   > RigVMVariableNode(Value=@n12.Item) {variable=Pelvis Control}
   n13 = RigVMVariableNode() {variable=Virtual Spine Joints}
   Prepare Control Shape_3 = RigVMFunctionReferenceNode(
    Scale Premultiplier=0.4
    Scale Factor=RigVMVariableNode() {variable=Control Scale}.Value
    Control Shape=RigVMVariableNode() {variable=Spline Fwd Proxy Control Shape}.Value
   )
   n14 = &RigUnit_HierarchyAddControlTransform(
    InitialValue=(Rotation=(X=0,Y=0,Z=0,W=1),Translation=(X=0,Y=0,Z=0),Scale3D=(X=1,Y=1,Z=1))
    Name=Spine Fwd Proxy
    OffsetSpace=GlobalSpace
    Settings="(InitialSpace=LocalSpace,bUsePreferredRotationOrder=False,PreferredRotationOrder=YZX,Limits=(LimitTranslationX=(bMinimum=False,bMaximum=False),LimitTranslationY=(bMinimum=False,bMaximum=False),LimitTranslationZ=(bMinimum=False,bMaximum=False),LimitPitch=(b ...(truncated)"
    OffsetTransform=&RigUnit_GetTransform(
     bInitial=true
     Item=RigVMDispatchNode(Index=0, Array=@n13.Value) {template="DISPATCH_RigVMDispatch_ArrayGetAtIndex(in Array,in Index,out Element)"}.Element
    ).Transform
    Settings.Shape=@"Prepare Control Shape_3".Result Control Shape
    Settings.Proxy.DrivenControls.0=RigVMVariableNode() {variable=Chest Control}.Value
    Parent=RigVMVariableNode() {variable=Connector Parent}.Value
    Settings.DisplayName=RigVMVariableNode() {variable=Spine Front Proxy Display Name}.Value
    ExecutePin=@RigVMFunction_Sequence.C
   ) {template="SpawnControl::Execute(in InitialValue,in Settings,in OffsetTransform,in OffsetSpace,in Parent,in Name,out Item)"}
   > RigVMVariableNode(Value=@n14.Item) {variable=Spine Fwd Proxy}
   n15 = RigVMVariableNode() {variable=Virtual Spine Joints}
   Prepare Control Shape_4 = RigVMFunctionReferenceNode(
    Scale Premultiplier=0.4
    Control Shape=RigVMVariableNode() {variable=Spline Bwd Proxy Control Shape}.Value
    Scale Factor=RigVMVariableNode() {variable=Control Scale}.Value
   )
   > n16 = &RigUnit_HierarchyAddControlTransform(
    InitialValue=(Rotation=(X=0,Y=0,Z=0,W=1),Translation=(X=0,Y=0,Z=0),Scale3D=(X=1,Y=1,Z=1))
    Name=Spine Bk Proxy
    OffsetSpace=GlobalSpace
    Settings="(InitialSpace=LocalSpace,bUsePreferredRotationOrder=False,PreferredRotationOrder=YZX,Limits=(LimitTranslationX=(bMinimum=False,bMaximum=False),LimitTranslationY=(bMinimum=False,bMaximum=False),LimitTranslationZ=(bMinimum=False,bMaximum=False),LimitPitch=(b ...(truncated)"
    OffsetTransform=&RigUnit_GetTransform(
     bInitial=true
     Item=RigVMDispatchNode(Index=2, Array=@n15.Value) {template="DISPATCH_RigVMDispatch_ArrayGetAtIndex(in Array,in Index,out Element)"}.Element
    ).Transform
    Settings.Shape=@"Prepare Control Shape_4".Result Control Shape
    Settings.Proxy.DrivenControls.0=RigVMVariableNode() {variable=Hip Control}.Value
    Parent=RigVMVariableNode() {variable=Connector Parent}.Value
    Settings.DisplayName=RigVMVariableNode() {variable=Spine Back Proxy Display Name}.Value
   ) {template="SpawnControl::Execute(in InitialValue,in Settings,in OffsetTransform,in OffsetSpace,in Parent,in Name,out Item)"}
   > RigVMVariableNode(Value=@n16.Item) {variable=Spine Bk Proxy}
   n17 = /Script/ControlRig.RigUnit_HierarchyAddAnimationChannelBool(
    InitialValue=false
    LimitsEnabled=()
    MaximumValue=True
    MinimumValue=false
    Name=Sec Controls Vis
    Parent=RigVMVariableNode() {variable=Hip Control}.Value
    ExecutePin=@RigVMFunction_Sequence.D
   ) {
    is=struct RigVMUnitNode
    template="SpawnAnimationChannel::Execute(in InitialValue,in MinimumValue,in MaximumValue,in LimitsEnabled,in Parent,in Name,out Item)"
   }
   Make_Array = RigVMDispatchNode(
    Values=((Type=None,Name="None"),(Type=None,Name="None"))
    Values.0=RigVMVariableNode() {variable=Spine 01 Control}.Value
    Values.1=RigVMVariableNode() {variable=Pelvis Control}.Value
   ) {template="DISPATCH_RigVMDispatch_ArrayMake(in Values,out Array)"}
   > /Script/ControlRig.RigUnit_SetChannelHosts(Hosts=@Make_Array.Array, Channel=@n17.Item) {is=struct RigVMUnitNode}
   > RigVMVariableNode(Value=@n17.Item) {variable=Sec Controls Vis}
   RigVMFunctionReturnNode(ExecuteContext=@n2.B)
  ]}
 New Function_ContainedGraph {nodes [
   n1 = RigVMFunctionEntryNode()
   RigVMFunction_Sequence = &RigVMFunction_Sequence(ExecuteContext=@n1.ExecuteContext)
   RigVMFunction_MathBoolNot = &RigVMFunction_MathBoolNot(Value=/Script/ControlRig.RigUnit_IsInteracting() {is=struct RigVMUnitNode}.bIsInteracting)
   RigVMFunction_MathBoolAnd = &RigVMFunction_MathBoolAnd(B=@RigVMFunction_MathBoolNot.Result, A=RigVMVariableNode() {variable=is Interacting}.Value)
   RigVMFunction_ControlFlowBranch = &RigVMFunction_ControlFlowBranch(Condition=@RigVMFunction_MathBoolAnd.Result, ExecuteContext=@RigVMFunction_Sequence.A)
   RigVMVariableNode(ExecuteContext=@RigVMFunction_ControlFlowBranch.True) {variable=is Interacting}
   > Update Proxy Offsets = RigVMFunctionReferenceNode()
   n2 = RigVMVariableNode() {variable=Spine Bk Proxy}
   n3 = RigVMVariableNode() {variable=Chest Control}
   > &RigUnit_SetTransform(
    Weight=1
    bInitial=false
    Value=/Script/ControlRig.RigUnit_ProjectTransformToNewParent(bChildInitial=True, bNewParentInitial=False, bOldParentInitial=True, Child=@n2.Value, NewParent=@n3.Value, OldParent=@n3.Value) {is=struct RigVMUnitNode}.Transform
    Item=@n2.Value
   )
   n4 = RigVMVariableNode() {variable=Hip Control}
   n5 = RigVMVariableNode() {variable=Spine Fwd Proxy}
   > &RigUnit_SetTransform(
    Weight=1
    bInitial=false
    Value=&RigUnit_ProjectTransformToNewParent(bChildInitial=True, bNewParentInitial=False, bOldParentInitial=True, NewParent=@n4.Value, OldParent=@n4.Value, Child=@n5.Value).Transform
    Item=@n5.Value
   )
   RigVMFunctionReturnNode(ExecuteContext=@RigVMFunction_Sequence.B)
  ]}
 New Function_ContainedGraph {nodes [
   n1 = RigVMFunctionEntryNode()
   Maximum = /Script/RigVM.RigVMFunction_MathIntMax(B=0, A=@n1.Index) {is=struct RigVMUnitNode, template="Maximum::Execute(in A,in B,out Result)"}
   Num = RigVMDispatchNode(Array=@n1.Name Array) {template="DISPATCH_RigVMDispatch_ArrayGetNum(in Array,out Num)"}
   Minimum = /Script/RigVM.RigVMFunction_MathIntMin(B=@Maximum.Result, A=@Num.Num) {is=struct RigVMUnitNode, template="Minimum::Execute(in A,in B,out Result)"}
   n2 = RigVMVariableNode() {variable=Name Array}
   Greater = /Script/RigVM.RigVMFunction_MathIntGreater(B=0, A=@Num.Num) {is=struct RigVMUnitNode, template="Greater::Execute(in A,in B,out Result)"}
   > RigVMFunctionReturnNode(Result=RigVMDispatchNode(
     True=RigVMDispatchNode(Index=@Minimum.Result, Array=@n2.Value) {template="DISPATCH_RigVMDispatch_ArrayGetAtIndex(in Array,in Index,out Element)"}.Element
     Condition=@Greater.Result
     False=RigVMVariableNode() {variable=Default}.Value
    ) {template="DISPATCH_RigVMDispatch_If(in Condition,in True,in False,out Result)"}.Result)
  ]}
 New Function_ContainedGraph {nodes [
   n1 = RigVMFunctionEntryNode()
   Get_Module_Metadata = RigVMDispatchNode(Name=Global Center Control COlor, NameSpace=Root, Default=@n1.Control Shape.Color) {template="DISPATCH_RigDispatch_GetModuleMetadata(in Name,in NameSpace,in Default,out Value,out Found)"}
   Interpolate = /Script/RigVM.RigVMFunction_MathColorLerp(B=@Get_Module_Metadata.Value, T=RigVMVariableNode() {variable=Root Module Color Weight}.Value, A=@n1.Control Shape.Color) {is=struct RigVMUnitNode, template="Interpolate::Execute(in A,in B,in T,out Result)"}
   Get_Module_Metadata_3 = RigVMDispatchNode(Default=1, Name=Global Control Scale, NameSpace=Root) {template="DISPATCH_RigDispatch_GetModuleMetadata(in Name,in NameSpace,in Default,out Value,out Found)"}
   Multiply = RigVMAggregateNode(
    C=@Get_Module_Metadata_3.Value
    A=RigVMVariableNode() {variable=Scale Factor}.Value
    B=RigVMVariableNode() {variable=Scale Premultiplier}.Value
   )
   n2 = RigVMVariableNode() {variable=Control Shape}
   Scale = /Script/RigVM.RigVMFunction_MathVectorScale(Factor=@Multiply.Result, Value=@n2.Value.Transform.Scale3D) {is=struct RigVMUnitNode, template="Scale::Execute(in Value,in Factor,out Result)"}
   > RigVMFunctionReturnNode(
    Result Control Shape=(bVisible=True,Name="Default",Color=(R=1,G=0,B=0,A=1),Transform=(Rotation=(X=0,Y=0,Z=0,W=1),Translation=(X=0,Y=0,Z=0),Scale3D=(X=1,Y=1,Z=1)))
    Result Control Shape.Name=@n1.Control Shape.Name
    Result Control Shape.bVisible=@n1.Control Shape.bVisible
    Result Control Shape.Color=@Interpolate.Result
    Result Control Shape.Transform.Scale3D=@Scale.Result
    Result Control Shape.Transform.Rotation=@n2.Value.Transform.Rotation
    Result Control Shape.Transform.Translation=@n2.Value.Transform.Translation
   )
  ]}
 New Function_ContainedGraph {nodes [
   RigVMFunctionEntryNode()
   RigUnit_GetTransform = &RigUnit_GetTransform(Item=(Type=Control,Name="Chest"), bInitial=False)
   > RigVMVariableNode(Value=&RigVMFunction_MathTransformMakeRelative(
     Global=@RigUnit_GetTransform.Transform
     Parent=&RigUnit_GetTransform(Item=(Type=Control,Name="Spine Fwd Proxy"), bInitial=False).Transform
    ).Local) {variable=Chest Local Transform Offset}
   > RigVMVariableNode(Value=&RigVMFunction_MathTransformMakeRelative(
     Parent=&RigUnit_GetTransform(Item=(Type=Control,Name="Spine Bk Proxy"), bInitial=False).Transform
     Global=&RigUnit_GetTransform(Item=(Type=Control,Name="Hips"), bInitial=False).Transform
    ).Local) {variable=Hip Local Transform Offset}
   > RigVMFunctionReturnNode()
  ]}
 New Function_ContainedGraph {nodes [
   n1 = RigVMFunctionEntryNode()
   Cast = /Script/RigVM.RigVMFunction_MathIntToFloat(Value=@n1.Index) {is=struct RigVMUnitNode, template="Cast::Execute(in Value,out Result)"}
   Subtract = /Script/RigVM.RigVMFunction_MathIntSub(B=0, A=@n1.Count) {is=struct RigVMUnitNode, template="Subtract::Execute(in A,in B,out Result)"}
   n2 = &RigVMFunction_MathIntToFloat(Value=@Subtract.Result) {template="Cast::Execute(in Value,out Result)"}
   LessEqual = /Script/RigVM.RigVMFunction_MathIntLessEqual(B=0, A=@n1.Count) {is=struct RigVMUnitNode, template="LessEqual::Execute(in A,in B,out Result)"}
   Get Scale Factor Profile Value = RigVMFunctionReferenceNode(Param=RigVMDispatchNode(True=0, False=&RigVMFunction_MathDoubleDiv(A=@Cast.Result, B=@n2.Result).Result, Condition=@LessEqual.Result) {template="DISPATCH_RigVMDispatch_If(in Condition,in True,in False,out Result)"}.Result)
   > RigVMFunctionReturnNode(Result=@"Get Scale Factor Profile Value".Result)
  ]}
 New Function_ContainedGraph {nodes [
   n1 = RigVMFunctionEntryNode()
   Multiply = &RigVMFunction_MathDoubleMul(
    B=&RigVMFunction_AnimEvalRichCurve(SourceMinimum=0, TargetMinimum=0, Value=@n1.Param, Curve=RigVMVariableNode() {variable=Control Scale Factor Profile}.Value).Result
    A=RigVMVariableNode() {variable=Control Scale}.Value
   )
   > RigVMFunctionReturnNode(Result=@Multiply.Result)
  ]}
 RigVMFunctionLibrary {nodes [
   Create Controls = RigVMCollapseNode(ExecuteContext=())
   Create Virtual Bones = RigVMCollapseNode(ExecuteContext=())
   Get Scale Factor Profile Value = RigVMCollapseNode(Param=0)
   Get Scale Factor Profile Value In For Loop = RigVMCollapseNode(Count=0, Index=0)
   On Proxy Interaction End = RigVMCollapseNode(ExecuteContext=())
   Prepare Control Shape = RigVMCollapseNode(
    Control Shape=(bVisible=True,Name="Default",Color=(R=1,G=0,B=0,A=1),Transform=(Rotation=(X=0,Y=0,Z=0,W=1),Translation=(X=0,Y=0,Z=0),Scale3D=(X=1,Y=1,Z=1)))
    Scale Factor=1
    Scale Premultiplier=1
   )
   Try Get Name From Array = RigVMCollapseNode(Name Array=())
   Update Proxy Offsets = RigVMCollapseNode(ExecuteContext=())
  ]}
 RigVMFunction_Sequence_1_ContainedGraph {nodes [
   n1 = RigVMFunctionEntryNode()
   RigVMFunction_Sequence = &RigVMFunction_Sequence(ExecuteContext=@n1.ExecuteContext)
   n2 = &RigVMFunction_Sequence(ExecuteContext=@RigVMFunction_Sequence.B)
   n3 = &RigVMFunction_Sequence(ExecuteContext=@n2.B)
   RigVMFunctionReturnNode(A=@RigVMFunction_Sequence.A, B=@n2.A, C=@n3.A, D=@n3.B)
  ]}
 RigVMFunction_Sequence_1_ContainedGraph {nodes [
   n1 = RigVMFunctionEntryNode()
   RigVMFunction_Sequence = &RigVMFunction_Sequence(ExecuteContext=@n1.ExecuteContext)
   n2 = &RigVMFunction_Sequence(ExecuteContext=@RigVMFunction_Sequence.B)
   n3 = &RigVMFunction_Sequence(ExecuteContext=@n2.B)
   RigVMFunctionReturnNode(A=@RigVMFunction_Sequence.A, B=@n2.A, C=@n3.A, D=@n3.B)
  ]}
 RigVMFunction_Sequence_1_ContainedGraph {nodes [
   n1 = RigVMFunctionEntryNode()
   RigVMFunction_Sequence = &RigVMFunction_Sequence(ExecuteContext=@n1.ExecuteContext)
   n2 = &RigVMFunction_Sequence(ExecuteContext=@RigVMFunction_Sequence.B)
   n3 = &RigVMFunction_Sequence(ExecuteContext=@n2.B)
   RigVMFunctionReturnNode(A=@RigVMFunction_Sequence.A, B=@n2.A, C=@n3.A, D=@n3.B)
  ]}
 RigVMFunction_Sequence_2_ContainedGraph {nodes [
   n1 = RigVMFunctionEntryNode()
   RigVMFunction_Sequence = &RigVMFunction_Sequence(ExecuteContext=@n1.ExecuteContext)
   n2 = &RigVMFunction_Sequence(ExecuteContext=@RigVMFunction_Sequence.B)
   n3 = &RigVMFunction_Sequence(ExecuteContext=@n2.B)
   n4 = &RigVMFunction_Sequence(ExecuteContext=@n3.B)
   n5 = &RigVMFunction_Sequence(ExecuteContext=@n4.B)
   RigVMFunctionReturnNode(A=@RigVMFunction_Sequence.A, B=@n2.A, C=@n3.A, D=@n4.A, E=@n5.A, F=@n5.B)
  ]}
 RigVMFunction_Sequence_2_ContainedGraph {nodes [
   n1 = RigVMFunctionEntryNode()
   n2 = &RigVMFunction_Sequence(ExecuteContext=@n1.ExecuteContext)
   n3 = &RigVMFunction_Sequence(ExecuteContext=@n2.B)
   RigVMFunctionReturnNode(A=@n2.A, B=@n3.A, C=@n3.B)
  ]}
 RigVMModel {nodes [
   RigUnit_BeginExecution = &RigUnit_BeginExecution()
   RigVMFunction_Sequence = RigVMAggregateNode(ExecuteContext=@RigUnit_BeginExecution.ExecutePin)
   n1 = RigVMVariableNode() {variable=Pelvis Control}
   n2 = RigVMVariableNode() {variable=Connector Spine Root}
   &RigUnit_SetTransform(
    Weight=1
    bInitial=false
    bPropagateToChildren=true
    Value=&RigUnit_ProjectTransformToNewParent(bChildInitial=True, bNewParentInitial=False, bOldParentInitial=True, NewParent=@n1.Value, OldParent=@n1.Value, Child=@n2.Value).Transform
    Item=@n2.Value
    ExecutePin=@RigVMFunction_Sequence.B
   )
   For_Each_1 = RigVMDispatchNode(Array=RigVMVariableNode() {variable=Spine Joints}.Value, ExecuteContext=@RigVMFunction_Sequence.C) {template="DISPATCH_RigVMDispatch_ArrayIterator(in Array,out Element,out Index,out Count,out Ratio)"}
   n3 = RigVMDispatchNode(Array=RigVMVariableNode() {variable=Controls}.Value, Index=@For_Each_1.Index) {template="DISPATCH_RigVMDispatch_ArrayGetAtIndex(in Array,in Index,out Element)"}
   > &RigUnit_SetTransform(
    Weight=1
    bInitial=False
    bPropagateToChildren=true
    Value=&RigUnit_ProjectTransformToNewParent(bChildInitial=True, bNewParentInitial=False, bOldParentInitial=True, NewParent=@n3.Element, OldParent=@n3.Element, Child=@For_Each_1.Element).Transform
    Item=@For_Each_1.Element
   )
   On Proxy Interaction End = RigVMFunctionReferenceNode(ExecuteContext=@RigVMFunction_Sequence.E)
   &RigUnit_ParentConstraint(
    AdvancedSettings=(InterpolationType=Average,RotationOrderForFilter=XZY)
    Parents=((Item=(Type=Control,Name="Spine 01"),Weight=0.25),(Item=(Type=Control,Name="Chest"),Weight=0.75))
    Weight=1
    Parents.0.Item=RigVMDispatchNode(Index=0, Array=RigVMVariableNode() {variable=Controls}.Value) {template="DISPATCH_RigVMDispatch_ArrayGetAtIndex(in Array,in Index,out Element)"}.Element
    Child=RigVMVariableNode() {variable=Spine 02 Null}.Value
    Parents.1.Item=RigVMVariableNode() {variable=Chest Control}.Value
    ExecutePin=@RigVMFunction_Sequence.A
   )
   Make_Array = RigVMDispatchNode(
    Values=((Type=None,Name="None"),(Type=None,Name="None"))
    Values.0=RigVMVariableNode() {variable=Spine 01 Control}.Value
    Values.1=RigVMVariableNode() {variable=Pelvis Control}.Value
   ) {template="DISPATCH_RigVMDispatch_ArrayMake(in Values,out Array)"}
   For_Each_2 = RigVMDispatchNode(Array=@Make_Array.Array, ExecuteContext=@RigVMFunction_Sequence.D) {template="DISPATCH_RigVMDispatch_ArrayIterator(in Array,out Element,out Index,out Count,out Ratio)"}
   GetAnimationChannelFromItem = /Script/ControlRig.RigUnit_GetBoolAnimationChannelFromItem(bInitial=False, Item=RigVMVariableNode() {variable=Sec Controls Vis}.Value) {is=struct RigVMUnitNode, template="GetAnimationChannelFromItem::Execute(out Value,in Item,in bInitial)"}
   > SetControlVisibility = /Script/ControlRig.RigUnit_SetControlVisibility(Item=@For_Each_2.Element, bVisible=@GetAnimationChannelFromItem.Value) {is=struct RigVMUnitNode, template="SetControlVisibility::Execute(in Item,in Pattern,in bVisible)"}
  ]}
 RigVMModel Backwards {nodes [
   n1 = /Script/ControlRig.RigUnit_InverseExecution() {is=struct RigVMUnitNode}
   n2 = RigVMVariableNode() {variable=Connector Spine Base}
   n3 = RigVMVariableNode() {variable=Hip Control}
   &RigUnit_SetTransform(
    Weight=1
    bInitial=false
    bPropagateToChildren=true
    Value=&RigUnit_ProjectTransformToNewParent(bChildInitial=True, bNewParentInitial=False, bOldParentInitial=True, NewParent=@n2.Value, OldParent=@n2.Value, Child=@n3.Value).Transform
    Item=@n3.Value
    ExecutePin=@n1.ExecutePin
   )
   n4 = RigVMVariableNode() {variable=Connector Spine Root}
   n5 = RigVMVariableNode() {variable=Pelvis Control}
   > &RigUnit_SetTransform(
    Weight=1
    bInitial=false
    bPropagateToChildren=false
    Value=&RigUnit_ProjectTransformToNewParent(bChildInitial=True, bNewParentInitial=False, bOldParentInitial=True, NewParent=@n4.Value, OldParent=@n4.Value, Child=@n5.Value).Transform
    Item=@n5.Value
   )
   n6 = RigVMDispatchNode(Index=2, Array=RigVMVariableNode() {variable=Controls}.Value) {template="DISPATCH_RigVMDispatch_ArrayGetAtIndex(in Array,in Index,out Element)"}
   n7 = RigVMVariableNode() {variable=Connector Spine Tip}
   > &RigUnit_SetTransform(
    Weight=1
    bInitial=false
    bPropagateToChildren=false
    Item=@n6.Element
    Value=&RigUnit_ProjectTransformToNewParent(bChildInitial=True, bNewParentInitial=False, bOldParentInitial=True, Child=@n6.Element, NewParent=@n7.Value, OldParent=@n7.Value).Transform
   )
   > &RigUnit_ParentConstraint(
    AdvancedSettings=(InterpolationType=Average,RotationOrderForFilter=XZY)
    Parents=((Item=(Type=Control,Name="Spine 01"),Weight=0.25),(Item=(Type=Control,Name="Chest"),Weight=0.75))
    Weight=1
    Parents.0.Item=RigVMDispatchNode(Index=0, Array=RigVMVariableNode() {variable=Controls}.Value) {template="DISPATCH_RigVMDispatch_ArrayGetAtIndex(in Array,in Index,out Element)"}.Element
    Child=RigVMVariableNode() {variable=Spine 02 Null}.Value
    Parents.1.Item=RigVMVariableNode() {variable=Chest Control}.Value
   )
   n8 = RigVMDispatchNode(Index=1, Array=RigVMVariableNode() {variable=Controls}.Value) {template="DISPATCH_RigVMDispatch_ArrayGetAtIndex(in Array,in Index,out Element)"}
   n9 = RigVMVariableNode() {variable=Connector Spine Mid}
   > &RigUnit_SetTransform(
    Weight=1
    bInitial=false
    bPropagateToChildren=false
    Item=@n8.Element
    Value=&RigUnit_ProjectTransformToNewParent(bChildInitial=True, bNewParentInitial=False, bOldParentInitial=True, Child=@n8.Element, NewParent=@n9.Value, OldParent=@n9.Value).Transform
   )
   RigVMRerouteNode(Value=())
  ]}
 RigVMModel Interaction Graph {nodes [
   n1 = /Script/ControlRig.RigUnit_InteractionExecution() {is=struct RigVMUnitNode}
   n2 = RigVMVariableNode(Value=true, ExecuteContext=@n1.ExecutePin) {variable=is Interacting}
   n3 = &RigUnit_IsInteracting()
   RigVMFunction_ControlFlowBranch = &RigVMFunction_ControlFlowBranch(
    Condition=RigVMDispatchNode(Array=@n3.Items, Element=RigVMVariableNode() {variable=Spine Fwd Proxy}.Value) {template="DISPATCH_RigVMDispatch_ArrayFind(in Array,in Element,out Index,out Success)"}.Success
    ExecuteContext=@n2.ExecuteContext
   )
   n4 = &RigVMFunction_ControlFlowBranch(
    Condition=RigVMDispatchNode(Array=@n3.Items, Element=RigVMVariableNode() {variable=Spine Bk Proxy}.Value) {template="DISPATCH_RigVMDispatch_ArrayFind(in Array,in Element,out Index,out Success)"}.Success
    ExecuteContext=@RigVMFunction_ControlFlowBranch.False
   )
   n5 = RigVMVariableNode() {variable=Spine Bk Proxy}
   &RigUnit_SetTransform(
    Weight=1
    bInitial=false
    Value=/Script/RigVM.RigVMFunction_MathTransformMakeAbsolute(
     Parent=&RigUnit_GetTransform(bInitial=false, Item=@n5.Value).Transform
     Local=RigVMVariableNode() {variable=Hip Local Transform Offset}.Value
    ) {is=struct RigVMUnitNode, template="Make Absolute::Execute(in Local,in Parent,out Global)"}.Global
    ExecutePin=@n4.True
    Item=RigVMVariableNode() {variable=Hip Control}.Value
   )
   > &RigUnit_ParentConstraint(
    AdvancedSettings=(InterpolationType=Average,RotationOrderForFilter=XZY)
    Child=(Type=Control,Name="Spine Fwd Proxy")
    Parents=((Item=(Type=Control,Name="Hip"),Weight=1))
    Weight=1
    bMaintainOffset=false
   )
   n6 = RigVMVariableNode() {variable=Spine Fwd Proxy}
   &RigUnit_SetTransform(
    Weight=1
    bInitial=false
    Value=&RigVMFunction_MathTransformMakeAbsolute(
     Parent=&RigUnit_GetTransform(bInitial=False, Item=@n6.Value).Transform
     Local=RigVMVariableNode() {variable=Chest Local Transform Offset}.Value
    ) {template="Make Absolute::Execute(in Local,in Parent,out Global)"}.Global
    ExecutePin=@RigVMFunction_ControlFlowBranch.True
    Item=RigVMVariableNode() {variable=Chest Control}.Value
   )
   > &RigUnit_ParentConstraint(
    AdvancedSettings=(InterpolationType=Average,RotationOrderForFilter=XZY)
    Child=(Type=Control,Name="Spine Bk Proxy")
    Parents=((Item=(Type=Control,Name="Chest"),Weight=1))
    Weight=1
    bMaintainOffset=false
   )
  ]}
 RigVMModel Rig1_1 {nodes [
   n1 = &RigUnit_PrepareForExecution()
   RigVMFunction_Sequence = RigVMAggregateNode(ExecuteContext=@n1.ExecutePin)
   RigVMFunction_Sequence_1 = RigVMAggregateNode(ExecuteContext=@RigVMFunction_Sequence.A)
   RigVMDispatchNode(Array=RigVMVariableNode() {variable=Spine Joints}.Value, ExecuteContext=@RigVMFunction_Sequence_1.A) {template="DISPATCH_RigVMDispatch_ArrayReset(io Array)"}
   > RigVMDispatchNode(Array=RigVMVariableNode() {variable=Virtual Spine Joints}.Value) {template="DISPATCH_RigVMDispatch_ArrayReset(io Array)"}
   > RigVMDispatchNode(Array=RigVMVariableNode() {variable=Controls}.Value) {template="DISPATCH_RigVMDispatch_ArrayReset(io Array)"}
   RigVMVariableNode(
    Value=/Script/ControlRig.RigUnit_ResolveConnector(Connector=(Type=Connector,Name="SpineRoot"), SkipSocket=False) {is=struct RigVMUnitNode}.Result
    ExecuteContext=@RigVMFunction_Sequence_1.B
   ) {variable=Connector Spine Root}
   > RigVMVariableNode(Value=&RigUnit_ResolveConnector(Connector=(Type=Connector,Name="Spine 1"), SkipSocket=False).Result) {variable=Connector Spine Base}
   > RigVMVariableNode(Value=&RigUnit_ResolveConnector(Connector=(Type=Connector,Name="Spine 2"), SkipSocket=False).Result) {variable=Connector Spine Mid}
   > RigVMVariableNode(Value=&RigUnit_ResolveConnector(Connector=(Type=Connector,Name="Spine 3"), SkipSocket=False).Result) {variable=Connector Spine Tip}
   > RigVMVariableNode(Value=&RigUnit_ResolveConnector(Connector=(Type=Connector,Name="Parent"), SkipSocket=False).Result) {variable=Connector Parent}
   RigVMVariableNode(
    Value=/Script/ControlRig.RigUnit_HierarchyGetChainItemArray(
     bIncludeEnd=True
     bIncludeStart=True
     bReverse=False
     Start=RigVMVariableNode() {variable=Connector Spine Base}.Value
     End=RigVMVariableNode() {variable=Connector Spine Tip}.Value
    ) {is=struct RigVMUnitNode}.Chain
    ExecuteContext=@RigVMFunction_Sequence_1.C
   ) {variable=Spine Joints}
   > Create Virtual Bones = RigVMFunctionReferenceNode()
   Create Controls_1 = RigVMFunctionReferenceNode(ExecuteContext=@RigVMFunction_Sequence.B)
   Update Proxy Offsets = RigVMFunctionReferenceNode(ExecuteContext=@RigVMFunction_Sequence.C)
  ]}
]
hierarchy [
 SpineRoot {type=Connector, loc=<0 0 0>, rot=<0 0 0 1>}
 Root {type=Socket, loc=<0 0 0>, rot=<0 0 0 1>}
 Spine 3 {type=Connector, loc=<0 0 0>, rot=<0 0 0 1>}
 Spine 1 {type=Connector, loc=<0 0 0>, rot=<0 0 0 1>}
 Parent {type=Connector, loc=<0 0 0>, rot=<0 0 0 1>}
 Spine 2 {type=Connector, loc=<0 0 0>, rot=<0 0 0 1>}
]
variables [
 Chest Control {type=FRigElementKey, default=(Type=Bone,Name="")}
 Chest Local Transform Offset {type=FTransform, default=(Rotation=(X=0,Y=0,Z=0,W=1),Translation=(X=0,Y=0,Z=0),Scale3D=(X=1,Y=1,Z=1))}
 Connector Parent {type=FRigElementKey, default=(Type=Connector,Name="")}
 Connector Spine Base {type=FRigElementKey, default=(Type=Connector,Name="")}
 Connector Spine Mid {type=FRigElementKey, default=(Type=None,Name="")}
 Connector Spine Root {type=FRigElementKey, default=(Type=Connector,Name="")}
 Connector Spine Tip {type=FRigElementKey, default=(Type=Connector,Name="")}
 Control Scale {type=double, default=1}
 Control Scale Factor Profile {
  type=FRuntimeFloatCurve
  default=(EditorCurveData=(Keys=((Value=3.25),(Time=0.5,Value=3),(Time=1,Value=4)),DefaultValue=340282346638528859811704183484516925440,PreInfinityExtrap=RCCE_Constant,PostInfinityExtrap=RCCE_Constant),ExternalCurve=None)
 }
 Controls {type="TArray<FRigElementKey>"}
 FKs Control Shapes {
  type=FRigUnit_HierarchyAddControl_ShapeSettings
  default=(bVisible=True,Name="Circle_Thick",Color=(R=1,G=1,B=0,A=1),Transform=(Rotation=(X=-0.7071,Y=0,Z=0,W=0.7071),Translation=(X=0,Y=0,Z=0),Scale3D=(X=1,Y=1,Z=1)))
 }
 Hip Control {type=FRigElementKey, default=(Type=Bone,Name="")}
 Hip Display Name {type=FName, default=Hips}
 Hip Local Transform Offset {type=FTransform, default=(Rotation=(X=0,Y=0,Z=0,W=1),Translation=(X=0,Y=0,Z=0),Scale3D=(X=1,Y=1,Z=1))}
 Pelvis Control {type=FRigElementKey, default=(Type=Bone,Name="")}
 Pelvis Control Shapes {
  type=FRigUnit_HierarchyAddControl_ShapeSettings
  default=(bVisible=True,Name="Hexagon_Thick",Color=(R=0,G=0.3,B=0,A=1),Transform=(Rotation=(X=-0.7071,Y=0,Z=0,W=0.7071),Translation=(X=0,Y=-0.5,Z=0),Scale3D=(X=1,Y=1,Z=1)))
 }
 Pelvis FK Display Name {type=FName, default=Pelvis FK}
 Pelvis Joint {type=FRigElementKey, default=(Type=Bone,Name="")}
 Root Module Color Weight {type=double, default=0}
 Sec Controls Vis {type=FRigElementKey, default=(Type=None,Name="")}
 Spine 01 Control {type=FRigElementKey, default=(Type=None,Name="")}
 Spine 02 {type=FRigElementKey, default=(Type=Bone,Name="")}
 Spine 02 Null {type=FRigElementKey, default=(Type=Bone,Name="")}
 Spine Back Proxy Display Name {type=FName, default=Spine Bk Proxy}
 Spine Bk Proxy {type=FRigElementKey, default=(Type=Bone,Name="")}
 Spine Control Shapes {
  type=FRigUnit_HierarchyAddControl_ShapeSettings
  default=(bVisible=True,Name="Hexagon_Thick",Color=(R=1,G=0.2,B=0,A=1),Transform=(Rotation=(X=-0.7071,Y=0,Z=0,W=0.7071),Translation=(X=0,Y=0.5,Z=0),Scale3D=(X=1,Y=1,Z=1)))
 }
 Spine Display Names {type="TArray<FName>", default=("Spine 01","Waist","Chest")}
 Spine Front Proxy Display Name {type=FName, default=Spine Fwd Proxy}
 Spine Fwd Proxy {type=FRigElementKey, default=(Type=Bone,Name="")}
 Spine Joints {type="TArray<FRigElementKey>"}
 Spline Bwd Proxy Control Shape {
  type=FRigUnit_HierarchyAddControl_ShapeSettings
  default=(bVisible=True,Name="Sphere_Solid",Color=(R=1,G=0.5,B=0,A=1),Transform=(Rotation=(X=0,Y=0,Z=0,W=1),Translation=(X=0,Y=-7,Z=25),Scale3D=(X=1,Y=1,Z=1)))
 }
 Spline Fwd Proxy Control Shape {
  type=FRigUnit_HierarchyAddControl_ShapeSettings
  default=(bVisible=True,Name="Sphere_Solid",Color=(R=1,G=0.5,B=0,A=1),Transform=(Rotation=(X=0,Y=0,Z=0,W=1),Translation=(X=0,Y=5,Z=23),Scale3D=(X=1,Y=1,Z=1)))
 }
 Virtual Spine Joints {type="TArray<FRigElementKey>"}
 is Interacting {type=bool, default=False}
]
